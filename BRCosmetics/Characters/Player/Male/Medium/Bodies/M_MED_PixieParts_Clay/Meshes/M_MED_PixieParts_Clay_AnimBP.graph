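# Animation Blueprint for the Pixie Parts Clay cosmetic body that adjusts upper-arm offsets to suit the equipped hat.
# It copies the pose from the attached parent mesh and runs it through a Control Rig fed two upper-arm transforms and a parachuting flag.
# When customization loads, it picks both transforms by matching "drill", "bolt" or "screw" in the hat mesh's name, with a fallback otherwise.
# The parachuting flag updates every animation tick.
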AnimBlueprint /BRCosmetics/Characters/Player/Male/Medium/Bodies/M_MED_PixieParts_Clay/Meshes/M_MED_PixieParts_Clay_AnimBP
target_skeleton=&/M_MED_PixieParts_Clay_Skeleton
graphs [
 AnimGraph {nodes [
   n1 = Control Rig!(
    Source=Copy Pose From Mesh() {bUseAttachedParent=true}
    Alpha
    parachuting:bool=Get parachuting() -> (parachuting:bool)
    upperarm_transform:Transform=Get upperarm_transform() -> (upperarm_transform:Transform)
    upperarm_transform_parachuting:Transform=Get upperarm_transform_parachuting() -> (upperarm_transform_parachuting:Transform)
   ) -> (Pose) {controlRigAssetReference {blueprintRigClass=&/M_MED_PixieParts_Clay_CtrlRig.M_MED_PixieParts_Clay_CtrlRig_C}}
   Output Pose(Result=@n1) {layerGroup=None}
  ]}
 EventGraph {nodes [
   Event BlueprintUpdateAnimation()
   > Set parachuting(parachuting:bool=Get bIsParachuteOpened()) -> (Output_Get:bool)
   n1 = Event OnCharacterCustomizationLoaded()
   n2 = GetObjectName(Object=Get SkeletalMeshAsset!(self:Skeletal Mesh Component Object Reference=Get Array Item with EFort Custom Part Type(Array=@n1.CustomMeshParts, PartType=Hat)) -> (SkeletalMeshAsset:Skeletal Mesh Object Reference))
   > n3 = Branch(Condition=Contains(SearchIn=@n2, Substring=drill))
   > Set upperarm_transform(upperarm_transform:Transform=MakeTransform(Location="1,0,7", Rotation="0, 0, 0")) -> (Output_Get:Transform)
   > Set upperarm_transform_parachuting(upperarm_transform_parachuting:Transform=MakeTransform(Location="1,-7,0", Rotation="0, 0, 0")) -> (Output_Get:Transform)
   n4 = Branch(execute=@n3.else, Condition=Contains(SearchIn=@n2, Substring=bolt))
   > Set upperarm_transform(upperarm_transform=MakeTransform(Location="4,0,3", Rotation="0, 0, 0"))
   > Set upperarm_transform_parachuting(upperarm_transform_parachuting=MakeTransform(Location="0,-4,0", Rotation="0, 0, 0"))
   n5 = Branch(execute=@n4.else, Condition=Contains(SearchIn=@n2, Substring=screw))
   > Set upperarm_transform(upperarm_transform=MakeTransform(Location="1,0,2", Rotation="0, 0, 0"))
   > Set upperarm_transform_parachuting(upperarm_transform_parachuting=MakeTransform(Location="1,-3,0", Rotation="0, 0, 0"))
   Set upperarm_transform(execute=@n5.else, upperarm_transform=MakeTransform(Location="2.5,0,2", Rotation="0, 0, 0"))
   > Set upperarm_transform_parachuting(upperarm_transform_parachuting=MakeTransform(Location="2,-3,0", Rotation="0, 0, 0"))
  ]}
]
variables [
 parachuting {type=boolean}
 upperarm_transform {type=Transform, default {location=<0 0 0>, rotation=<rot 0 0 0>, scale=<1 1 1>}}
 upperarm_transform_parachuting {type=Transform, default {location=<0 0 0>, rotation=<rot 0 0 0>, scale=<1 1 1>}}
]
